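# Material function that overlays cracks on incoming material attributes, driven by a Cracks scalar or, with the IceShellCracks switch, by code-set custom damage percentage.
# Five UV scales of one crack mask combine through Min, with the three finest revealed progressively as the amount rises.
# The mask drives HeightLerps for color and emissive, adjusting roughness and subsurface, plus an added crack normal.
# A CracksTrue switch gates the final blend; the described blue-vertex-channel painting is unconnected.
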MaterialFunction /Figment_S07/LM_7_00/Athena/WinterArt/Materials/MF_Cracks
graph {nodes [
  n1 = ScalarParameter() {group=Code, parameterName=Code_CustomDamagePercentage}
  n2 = LinearInterpolate(A=ScalarParameter() {defaultValue=0.05, parameterName=CrackMin}, B=ScalarParameter() {defaultValue=3, parameterName=CrackMax}, Alpha=@n1)
  n3 = StaticSwitchParameter(
   True=MaterialFunctionCall(Speed=Constant() {r=0.3333}) {materialFunction=&TimeWithSpeedVariable}.Frac Time
   False=@n2
  ) {group=CustomDamage, parameterName=Debug custom damage animation}
  Clamp(None=Multiply(A=@n3) {constB=0.5})
  Constant() {r=0.6971}
  n4 = FunctionInput() {inputType=FunctionInput_MaterialAttributes}
  n5 = GetMaterialAttributes(None=@n4) {attributeGetTypes [
    #1
    #2
    #3
    #4
    #5
    #6
   ]}
  n6 = Multiply(A=TextureCoordinate(), B=ScalarParameter() {defaultValue=3, desc=Mesh UV's, group=Cracks, parameterName=ScaleCracks})
  n7 = TextureSample(UVs=@n6, Tex=TextureObject() {samplerType=SAMPLERTYPE_Normal, texture=/Figment_S07/LM_7_00/Athena/WinterArt/Textures/Cracks_05_N}) {
   samplerSource=SSM_Wrap_WorldGroupSettings
   samplerType=SAMPLERTYPE_Normal
   texture=&Cracks_05_N
  }
  n8 = Normalize(VectorInput=Multiply(A=@n7.RGB, B=Constant3Vector() {constant=<rgba 3 3 1 0>}))
  n9 = Constant() {r=1.5}
  n10 = TextureObject() {samplerType=SAMPLERTYPE_Masks, texture=/Figment_S07/LM_7_00/Athena/WinterArt/Textures/Cracks_05_M}
  n11 = Multiply(A=@n6, B=Constant() {r=1.5}) {constB=0.5}
  n12 = Multiply(A=@n11, B=Constant() {r=1.5}) {constB=0.5}
  n13 = ScalarParameter() {
   desc="Scalar*Vertex Blue Channel Scale \"Cracks\" above 1 to get brighter look before destruction."
   group=Cracks
   parameterName=Cracks
   sliderMax=3
  }
  n14 = StaticSwitchParameter(True=@n3, False=@n13) {parameterName=IceShellCracks}
  n15 = Add(
   A=TextureSample(UVs=@n12, Tex=@n10) {samplerSource=SSM_Wrap_WorldGroupSettings, samplerType=SAMPLERTYPE_Masks}.RGB
   B=OneMinus(None=Saturate(None=@n14))
  )
  n16 = Multiply(A=@n12, B=Constant() {r=1.5}) {constB=0.5}
  n17 = Saturate(None=Add(A=@n14) {constB=-1})
  n18 = Multiply(A=@n16, B=Constant() {r=1.5}) {constB=0.5}
  n19 = Saturate(None=Add(A=@n14) {constB=-2})
  n20 = Min(
   A=Add(A=TextureSample(UVs=@n16, Tex=@n10) {samplerSource=SSM_Wrap_WorldGroupSettings, samplerType=SAMPLERTYPE_Masks}.RGB, B=OneMinus(None=@n17))
   B=Add(A=TextureSample(UVs=@n18, Tex=@n10) {samplerSource=SSM_Wrap_WorldGroupSettings, samplerType=SAMPLERTYPE_Masks}.RGB, B=OneMinus(None=@n19))
  )
  n21 = Min(
   A=TextureSample(UVs=@n6, Tex=@n10) {samplerSource=SSM_Wrap_WorldGroupSettings, samplerType=SAMPLERTYPE_Masks}.R
   B=Min(A=TextureSample(UVs=@n11, Tex=@n10) {samplerSource=SSM_Wrap_WorldGroupSettings, samplerType=SAMPLERTYPE_Masks}.R, B=Min(A=@n15, B=@n20))
  )
  n22 = MaterialFunctionCall(
   A=@n5.EmissiveColor
   B=Desaturation(None=Multiply(A=@n5.EmissiveColor, B=@n9) {constB=3}, Fraction=Constant() {r=-0.25})
   Transition Phase=Constant() {r=0.35}
   Height Texture=@n21
   Contrast=Constant() {r=-0.75}
  ) {materialFunction=&HeightLerp}
  n23 = Constant() {r=2}
  n24 = VertexNormalWS()
  n25 = MaterialFunctionCall(
   A=@n5.BaseColor
   B=VectorParameter() {group=Cracks, parameterName=CracksColor, defaultValue=<rgba 0.7344 0.7344 0.7344 0>}.RGB
   Transition Phase=LinearInterpolate(A=Constant() {r=0.7}, B=Constant() {r=1}, Alpha=Fresnel(ExponentIn=@n23, Normal=@n24) {baseReflectFraction=0.1, exponent=1})
   Height Texture=@n21
   Contrast=Constant() {r=-3}
  ) {materialFunction=&HeightLerp}
  n26 = SetMaterialAttributes!(
   MaterialAttributes=@n5.MaterialAttributes
   Normal=Add(A=@n5.Normal, B=Transform(None=@n8))
   EmissiveColor=@n22.Results
   BaseColor=@n25.Results
   SubsurfaceColor=Multiply(A=@n5.SubsurfaceColor, B=OneMinus(None=@n22.Alpha))
   Roughness=LinearInterpolate(A=@n5.Roughness, Alpha=@n25.Alpha) {constA=1}
  ) {attributeSetTypes [
    #1
    #2
    #3
    #4
    #5
   ]}
  n27 = ScalarParameter() {
   desc="Scalar*Vertex Blue Channel Scale \"Cracks\" above 1 to get brighter look before destruction."
   group=Cracks
   parameterName=Cracks
   sliderMax=3
  }
  n28 = Clamp(None=Divide(A=@n27)) {maxDefault=1.5}
  n29 = StaticSwitchParameter(True=@n2, False=@n28) {parameterName=IceShellCracks}
  FunctionOutput(None=StaticSwitchParameter(True=BlendMaterialAttributes(A=@n4, B=@n26, Alpha=@n29), False=@n4) {desc="Use Blue Channel to paint cracks. ", group=Cracks, parameterName=CracksTrue})
  LinearInterpolate()
  Multiply(A=@n28, B=VertexColor().B)
  Multiply(A=@n1) {constB=3}
  Multiply(A=@n21)
  Power(Base=OneMinus()) {constExponent=4}
  Saturate(None=Add(A=@n21))
  ScalarParameter() {group=Code, parameterName=Code_CustomDamagePercentage}
 ]}
