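# MF_GameWideShared is a material function that layers shared gameplay effects, hit deformation and Creative move-tool handling, onto incoming material attributes.
# Fortnite_HitEffects takes z_code parameters for hit animation, destruction, deformation and distance-cull fade.
# Only its Hit And Shrink output is used, replacing the incoming World Position Offset rather than adding to it.
# The modified attributes then pass through MF_CreativeMoveTool to the output.
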MaterialFunction /Figment_S07/LM_7_00/Athena/WinterArt/Materials/MF_GameWideShared
graph {nodes [
  n1 = FunctionInput() {inputType=FunctionInput_MaterialAttributes}
  n2 = MaterialFunctionCall!(
   CrackDamageShader
   Code_HitAnimation=ScalarParameter() {group=z_code, parameterName=Code_HitAnimation}
   Code_DestructionAmount=ScalarParameter() {group=z_code, parameterName=Code_DestructionAmount}
   Code_UseUvForDamage
   Code_DeformationVector=VectorParameter() {group=z_code, parameterName=Code_DeformationVector, defaultValue=<rgba 0 0 1 0>}.RGB
   Code_UseHitEffects=StaticBoolParameter() {defaultValue=true, group=z_code, parameterName=Code_UseHitEffects}
   WSDeformationCenter=VectorParameter() {group=z_code, parameterName=WSDeformationCenter}.RGB
   WSDamageRadius=ScalarParameter() {defaultValue=1, group=z_code, parameterName=WSDamageRadius}
   Diffuse
   Code_Emissive
   ScalarModDamageCracks
   ScalarOffsetforDamageCracks
   ExposeLerpFunctionForInvertingTheHitDistanceCalc
   InvertDamageFalloff
   WorldPosition
   DistanceCullFadeAnimationOffset=ScalarParameter() {group=z_code, parameterName=DistanceCullFadeAnimationOffset}
  ) {materialFunction=/Figment_S07/LM_7_00/Packages/jl_tests/Fortnite_HitEffects}
  FunctionOutput(None=MaterialFunctionCall!(MA In=SetMaterialAttributes!(MaterialAttributes=@n1, WorldPositionOffset=@n2.Hit And Shrink WPO) {attributeSetTypes [#1]}) {materialFunction=/Figment_S07/LM_7_00/Weapons/Prototype/ObjectMover/Materials/MF_CreativeMoveTool}.MA Out)
 ]}
