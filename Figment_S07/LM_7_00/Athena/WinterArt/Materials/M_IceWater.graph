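# M_IceWater is a Default Lit material for ice water, part of the Athena winter art set.
# It uses material attributes, and its graph contains a single call to the MF_FrozenLake material function, which holds the actual frozen-lake shading logic.
# Automatic usage flag setting in the editor is disabled.
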Material /Figment_S07/LM_7_00/Athena/WinterArt/Materials/M_IceWater
graph {
 properties {bAutomaticallySetUsageInEditor=false, bUseMaterialAttributes=true, pixelDepthOffsetMode=PDOM_Legacy, usedShadingModels=Default Lit}
 nodes [MaterialFunctionCall() {materialFunction=&/MF_FrozenLake}]
}
