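# Material function that builds a screen-space, time-varying blue-noise dither for TemporalAA, aimed at masked translucency and anisotropic materials.
# Noise scaled by a Random input offsets an Alpha Threshold, producing default, 64-frame and golden-ratio result outputs.
# Raw noise outputs accompany these, including an 8-frame TSR variant.
# It contains 128-pixel screen-aligned UVs, custom frame-index cycles, a golden-ratio frame offset and a scalar blue-noise source.
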MaterialFunction /Game/Environments/Asteria/Sets/NeonCity/Techart/Screen/MaterialFunctions/DitherTemporalAASpatioTemporalBlue
graph {
 properties {description="Stipple pattern/Dither  in screen space and time to work well with TemporalAA for things like masked translucency or anisotropic materials"}
 nodes [
  n1 = MaterialFunctionCall!(TextureResolution=Constant2Vector() {g=128, r=128}) {materialFunction=/Engine/Functions/Engine_MaterialFunctions02/Texturing/ScreenAlignedPixelToPixelUVs}
  AppendVector(A=@n1.Result, B=Custom!(None) {code="return View.StateFrameIndexMod8;", outputType=CMOT_Float1})
  AppendVector(A=@n1.Result, B=Custom() {code="return View.StateFrameIndex % 64;", outputType=CMOT_Float1})
  Constant()
  Frac(None=Add(B=Custom() {code="return (View.StateFrameIndex * 1.61803398875) ;", outputType=CMOT_Float1}))
  n2 = FunctionInput() {bUsePreviewValueAsDefault=true, inputName=Alpha Threshold, inputType=FunctionInput_Scalar, sortPriority=-1, previewValue=<0.5 0 0 1>}
  n3 = FunctionInput() {
   bUsePreviewValueAsDefault=true
   description="0=no random 1:random(default)"
   inputName=Random
   inputType=FunctionInput_Scalar
   previewValue=<1 0 0 1>
  }
  n4 = Multiply(A=NamedRerouteUsage(), B=@n3)
  FunctionOutput(None=Add(A=Add(A=@n2, B=@n4) {constB=0.5}) {constB=-0.5})
  n5 = NamedRerouteUsage()
  FunctionOutput(None=@n5) {outputName=RawNoise64Frames}
  FunctionOutput(None=NamedRerouteUsage()) {outputName=RawNoise8FramesTSR}
  n6 = Multiply(A=@n5, B=@n3)
  FunctionOutput(None=Add(A=Add(A=@n2, B=@n6) {constB=0.5}) {constB=-0.5}) {outputName=Result64}
  FunctionOutput(None=NamedRerouteUsage()) {outputName=RawFakeTemporalGoldenRatioBlueNoise}
  n7 = Multiply(A=@n5, B=@n3)
  FunctionOutput(None=Add(A=Add(A=@n7, B=@n2)) {constB=-0.5}) {outputName=ResultFakeTemporalGoldenRatioBlueNoise}
  Multiply()
  NamedRerouteDeclaration(None=ScalarBlueNoise()) {name=BN, nodeColor=<rgba 1 0.9609 0 1>}
 ]
}
