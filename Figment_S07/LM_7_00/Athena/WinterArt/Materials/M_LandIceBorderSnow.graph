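# Masked snow material, apparently for ice border meshes on the Athena winter landscape, that gives the snow strip broken, soft edges.
# A UV V gradient band, modulated by world-aligned Perlin noise and snow mask alpha, drives opacity through temporal blue-noise dithering.
# The shared Athena snow function supplies attributes, multiplied by static-mesh biome color, with a world-aligned snow normal blended in.
# Low quality uses only that normal and the undithered mask.
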Material /Figment_S07/LM_7_00/Athena/WinterArt/Materials/M_LandIceBorderSnow
graph {
 properties {bAutomaticallySetUsageInEditor=false, bUseMaterialAttributes=true, blendMode=BLEND_Masked, usedShadingModels=Default Lit}
 nodes [
  n1 = ScalarParameter() {defaultValue=1000}
  n2 = StaticBool() {value=true}
  n3 = MaterialFunctionCall(
   TextureObject=TextureObject() {samplerType=SAMPLERTYPE_Masks, texture=/Game/Athena/WinterArt/Textures/T_Snow01_M}
   TextureSize=@n1
   Export Float 4=@n2
  ) {materialFunction=&WorldAlignedTexture}
  ComponentMask(None=@n3.Z Texture) {r=true}
  n4 = ScalarParameter() {defaultValue=0.4, parameterName=Param_4}
  n5 = ComponentMask(None=TextureCoordinate()) {g=true}
  n6 = Saturate(None=Min(A=InverseLinearInterpolate(B=@n4, Value=@n5), B=InverseLinearInterpolate(B=@n4, Value=@n5) {constA=1}))
  n7 = ComponentMask(None=MaterialFunctionCall(
    TextureObject=TextureObject() {samplerType=SAMPLERTYPE_LinearColor, texture=/Game/Global/Textures/Default/Tileable_Patterns/T_PerlinNoise_Uncompressed}
    TextureSize=ScalarParameter() {defaultValue=1561.5442, parameterName=Param_1}
    Export Float 4=@n2
   ) {materialFunction=&WorldAlignedTexture}.Z Texture) {r=true}
  n8 = LinearInterpolate(
   A=ScalarParameter() {defaultValue=2.077, parameterName=Param_22}
   B=ScalarParameter() {defaultValue=1.2452, parameterName=Param_32}
   Alpha=@n7
  ) {constA=0.75, constB=2}
  n9 = ScalarParameter() {parameterName=Param_2}
  n10 = ScalarParameter() {defaultValue=0.5146, parameterName=Param_3}
  n11 = Subtract(
   A=Subtract(A=Multiply(A=@n6, B=@n8) {constB=1.5}, B=LinearInterpolate(A=@n9, B=@n10, Alpha=@n7) {constA=0.75, constB=2})
   B=ComponentMask(None=@n3.Z Texture) {a=true}
  )
  n12 = Subtract(A=@n11) {constB=0.33}
  n13 = Add(A=Abs(None=DDX(Value=@n12)), B=Abs(None=DDY(Value=@n12)))
  Divide(A=@n12, B=Multiply(A=Max(A=@n13) {constB=0.0001}) {constB=3})
  n14 = ComponentMask(None=@n3.Z Texture) {g=true}
  LinearInterpolate(Alpha=Saturate(None=Multiply(A=@n14) {constB=3})) {constA=0.8, constB=0.2}
  MaterialFunctionCall() {materialFunction=/Figment_S07/LM_7_00/Athena/Environments/Landscape/MaterialFunctions/Snow/MF_Athena_Snow_01}
  n15 = Add(A=Multiply(A=@n12) {constB=10}) {constB=0.5}
  MaterialFunctionCall!(Alpha Threshold=@n15, Random) {materialFunction=/Game/Environments/Asteria/Sets/NeonCity/Techart/Screen/MaterialFunctions/DitherTemporalAABlue}
  n16 = MaterialFunctionCall!(UseLandscapeCoordsOnSwitch) {materialFunction=/FigmentEnvironment/LM_8_1_1/Athena/Environments/Landscape/MaterialFunctions/Snow/MF_Athena_Snow_01}
  n17 = MaterialFunctionCall(
   TextureObject=TextureObject() {samplerType=SAMPLERTYPE_Normal, texture=/Game/Global/Textures/Default/Tileable_Patterns/T_Snow01_N}
   TextureSize=@n1
  ) {materialFunction=&WorldAlignedNormal}
  SetMaterialAttributes!(
   MaterialAttributes=MaterialFunctionCall!(
    In=@n16.Material
    Color=MaterialFunctionCall!(Material) {materialFunction=/FigmentEnvironment/LM_8_1_1/Athena/Environments/Materials/MF_StaticMeshBiomeColor}.Color
   ) {materialFunction=/FigmentEnvironment/LM_8_1_1/Athena/Environments/Landscape/MaterialFunctions/MF_Athena_Material_BaseMultiply}.Result
   Normal=QualitySwitch(
    Default=MaterialFunctionCall(BaseNormal=@n16.Normal, AdditionalNormal=@n17.Z Texture) {materialFunction=&BlendAngleCorrectedNormals}.Result
    Low=@n17.Z Texture
   )
   Opacity Mask=QualitySwitch(
    Default=MaterialFunctionCall!(Alpha Threshold=@n15, Random) {materialFunction=/Game/Environments/Asteria/Sets/NeonCity/Techart/Screen/MaterialFunctions/DitherTemporalAASpatioTemporalBlue}.Result
    Low=@n11
   )
  ) {attributeSetTypes [#1, #2]}
 ]
}
