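# Shared winter material function that supplies time-of-day snow and ice colors as two outputs, SnowColor and IceColor.
# Morning, day, evening and night color parameters for each surface feed the Color_Time_of_Day function.
# Snow uses that result in realtime and swaps to a constant white in material proxy bakes via MaterialProxyReplace.
# Ice lerps from a base IceColor parameter to its time-of-day color by IceTintValue, fully tinted by default.
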MaterialFunction /Figment_S07/LM_7_00/Athena/WinterArt/Materials/MF_SharedWinter
graph {nodes [
  n1 = VectorParameter() {group=Colors, parameterName=SnowMorning, sortPriority=5, defaultValue=<rgba 0.9386 0.9284 0.9635 0>}
  n2 = VectorParameter() {group=Colors, parameterName=SnowDay, sortPriority=6, defaultValue=<rgba 1 1 1 0>}
  n3 = VectorParameter() {group=Colors, parameterName=SnowEvening, sortPriority=7, defaultValue=<rgba 0.7561 0.8732 0.974 0>}
  n4 = VectorParameter() {group=Colors, parameterName=SnowNight, sortPriority=8, defaultValue=<rgba 0.7682 0.762 0.8073 0>}
  FunctionOutput(None=MaterialProxyReplace(
    Realtime=MaterialFunctionCall!(Color_Morning=@n1.RGB, Color_Day=@n2.RGB, Color_Evening=@n3.RGB, Color_Night=@n4.RGB) {materialFunction=/Figment_S07/LM_7_00/Effects/Fort_Effects/Materials/_Functions/Color_Time_of_Day}.Result
    MaterialProxy=Constant() {r=1}
   )) {outputName=SnowColor}
  n5 = VectorParameter() {group=Colors, parameterName=IceMorning, sortPriority=1, defaultValue=<rgba 0.1199 0.2582 0.4 1>}
  n6 = VectorParameter() {group=Colors, parameterName=IceDay, sortPriority=2, defaultValue=<rgba 0.1199 0.2582 0.4 1>}
  n7 = VectorParameter() {group=Colors, parameterName=IceEvening, sortPriority=3, defaultValue=<rgba 0.0889 0.1977 0.4 1>}
  n8 = VectorParameter() {group=Colors, parameterName=IceNight, sortPriority=4, defaultValue=<rgba 0.1199 0.2582 0.4 1>}
  FunctionOutput(None=LinearInterpolate(
    A=VectorParameter() {group=Colors, parameterName=IceColor, defaultValue=<rgba 0.0877 0.2373 0.276 1>}.RGB
    B=MaterialFunctionCall(Color_Morning=@n5.RGB, Color_Day=@n6.RGB, Color_Evening=@n7.RGB, Color_Night=@n8.RGB) {materialFunction=&Color_Time_of_Day}.Result
    Alpha=ScalarParameter() {defaultValue=1, group=Colors, parameterName=IceTintValue}
   ) {constAlpha=0.25}) {outputName=IceColor}
 ]}
